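# Additive unlit Niagara sprite material that draws a glowing halo ring of animated spiral lines.
# Two radially mapped samples of a spiral line texture pan in opposite directions with separate tiling, then sum and get masked by subtracted sphere masks.
# A panning rainbow texture supplies color through desaturation, tint, intensity, particle color and an emissive adjustment function.
# Particle alpha and depth fade modulate the ring opacity.
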Material /Game/Effects/Niagara/MasterMaterials/Ring/M_Spiral_HaloRing
graph {
 properties {bUsedWithNiagaraSprites=true, blendMode=BLEND_Additive, pixelDepthOffsetMode=PDOM_Legacy, shadingModel=MSM_Unlit, usedShadingModels=Unlit}
 nodes [
  n1 = MaterialFunctionCall() {materialFunction=&VectorToRadialValue}
  n2 = Panner(Coordinate=@n1.Radial Coordinates) {speedY=-0.5}
  n3 = Multiply(
   A=Desaturation(
    None=TextureSample(UVs=@n2) {texture=&T_Tile_RainbowColor}.RGB
    Fraction=ScalarParameter() {parameterName=Desaturation}
   )
   B=VectorParameter() {parameterName=Color, defaultValue=<rgba 2.5 4.4459 5 1>}.RGB
  )
  n4 = ScalarParameter() {defaultValue=20, parameterName=Intensity}
  n5 = ParticleColor()
  MaterialFunctionCall(Emissive=Multiply(A=Multiply(A=@n3, B=@n4), B=@n5.RGB)) {materialFunction=&MF_FX_EmissiveAdjustments}
  n6 = MaterialFunctionCall(Float2=@n1.Radial Coordinates) {materialFunction=&BreakOutFloat2Components}
  n7 = ScalarParameter() {defaultValue=2.7, parameterName=TextureA_VMultiply}
  n8 = AppendVector(
   A=Multiply(A=@n6.R, B=ScalarParameter() {defaultValue=8, parameterName=TextureA_UTiling}) {constB=8}
   B=Subtract(A=Multiply(A=@n6.G, B=@n7) {constB=2.7}, B=ScalarParameter() {defaultValue=1.85, parameterName=TextureA_VSubtract}) {constB=1.85}
  )
  n9 = MaterialFunctionCall(Float2=MaterialFunctionCall() {materialFunction=&VectorToRadialValue}.Radial Coordinates) {materialFunction=&BreakOutFloat2Components}
  n10 = ScalarParameter() {defaultValue=3.1, parameterName=TextureB_VMultiply}
  n11 = AppendVector(
   A=Multiply(A=@n9.R, B=ScalarParameter() {defaultValue=4, parameterName=TextureB_UTiling}) {constB=4}
   B=Subtract(A=Multiply(A=@n9.G, B=@n10) {constB=3.1}, B=ScalarParameter() {defaultValue=2, parameterName=TextureB_VSubtract}) {constB=2}
  )
  n12 = Add(
   A=TextureSample!(UVs=Panner(Coordinate=@n8) {speedX=0.2}, Tex, Level, Apply View MipBias) {mipValueMode=TMVM_MipLevel, texture=/Game/Effects/Fort_Effects/Textures/T_Emote_Spiral_Line_02}.G
   B=TextureSample(UVs=Panner(Coordinate=@n11) {speedX=-0.1}) {mipValueMode=TMVM_MipLevel, texture=&T_Emote_Spiral_Line_02}.G
  )
  n13 = TextureCoordinate()
  n14 = Constant() {r=0.5}
  n15 = Constant() {r=0.5}
  n16 = Constant()
  n17 = Subtract(A=Power(Base=SphereMask(A=@n13, B=@n14, Radius=@n15, Hardness=@n16)), B=SphereMask(A=@n13, B=@n14, Radius=Constant() {r=0.3}, Hardness=@n16))
  Multiply(A=@n5.A, B=DepthFade(Opacity=Multiply(A=@n12, B=@n17), FadeDistance=ScalarParameter() {defaultValue=5, parameterName=FadeDistance}))
 ]
}
