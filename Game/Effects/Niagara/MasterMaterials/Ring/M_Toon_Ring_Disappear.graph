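# Translucent unlit Niagara sprite material that draws a toon-style ring that thins and dissolves away.
# Subtracting an offset inner SphereMask from an outer one forms the ring, with dynamic parameter Dis01 enlarging and recentering the inner mask.
# Dis02 drives a HeightLerp over a dissolve texture, rounded for hard cutout edges.
# Particle color feeds emissive through MF_FX_EmissiveAdjustments, and particle alpha scales the mask.
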Material /Game/Effects/Niagara/MasterMaterials/Ring/M_Toon_Ring_Disappear
graph {
 properties {bUsedWithNiagaraSprites=true, blendMode=BLEND_Translucent, pixelDepthOffsetMode=PDOM_Legacy, shadingModel=MSM_Unlit, usedShadingModels=Unlit}
 nodes [
  n1 = ParticleColor()
  MaterialFunctionCall(Emissive=@n1.RGB) {materialFunction=&MF_FX_EmissiveAdjustments}
  n2 = ScalarParameter() {defaultValue=1, parameterName=Hardness}
  n3 = SphereMask(A=TextureCoordinate(), B=Constant() {r=0.5}, Hardness=@n2) {attenuationRadius=0.5}
  n4 = DynamicParameter() {defaultValue=<rgba 0 0 1 1>, paramNames [Dis01,Dis02,Param3,Param4]}
  n5 = SphereMask(
   A=TextureCoordinate()
   B=LinearInterpolate(A=Constant2Vector() {g=0.5, r=0.6}, B=Constant2Vector() {g=0.5, r=0.52}, Alpha=@n4.Dis01)
   Radius=LinearInterpolate(Alpha=@n4.Dis01) {constA=0.38, constB=0.49}
   Hardness=@n2
  ) {attenuationRadius=0.4}
  n6 = TextureCoordinate()
  n7 = Multiply(A=MaterialFunctionCall(
    Transition Phase=OneMinus(None=@n4.Dis02)
    Height Texture=TextureSample(UVs=@n6) {texture=/Game/CharClasses/Outlander/Gadgets/AntiMaterialCharge/Materials/T_AntiMaterialCharge_Dissolve_01}.R
   ) {materialFunction=&HeightLerp}.Alpha)
  Multiply(A=@n1.A, B=Multiply(A=Subtract(A=@n3, B=@n5), B=Round(None=@n7)))
 ]
}
